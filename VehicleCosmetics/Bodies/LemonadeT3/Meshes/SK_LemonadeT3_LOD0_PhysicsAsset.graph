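# Physics asset for the LemonadeT3 vehicle cosmetic body's LOD0 skeletal mesh, giving it a simplified physics shape.
# A single body on the frame bone holds one large capsule, offset from the bone origin and rotated nearly 90 degrees on two axes.
# It has no constraints and no other bodies, so the asset defines no articulated parts.
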PhysicsAsset /VehicleCosmetics/Bodies/LemonadeT3/Meshes/SK_LemonadeT3_LOD0_PhysicsAsset
bodies {frame [{shapeType=Capsule, length=238.8678, radius=120.1525, center=<28.7238 -84.3991 -0.0216>, rotation=<rot 0 -88.2941 -89.894>}]}
constraints []
